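; Quest ability for Season 36 Weekly W02 Q01 that credits a player elimination made shortly after a CherrySmoke dual revolvers dodge roll.
; Each dodge gameplay event (tagged DoubleJumpActivated) opens a filtered elimination listener on the owner, closed after the Time Limit delay.
; An elimination in that window broadcasts the DodgeElim custom verb and ends the ability.
; Removal cancels both waits; the b_validElim flag is set but never read here.
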(event Ability|EventActivateAbility
  (bind _returnvalue (Ability|GetOwningActorfromActorInfo))
  (bind _asynctaskproxy (Ability|Async|WaitGameplayEventToActor (Ability|GetOwningActorfromActorInfo) "(TagName=\"Granted.CherrySmokeDualRevolvers.DoubleJumpActivated\")")
    (:then
      (Variables|Default|Setasync_dodge _asynctaskproxy))
    (:EventReceived
      ; For 5 seconds after the Dodge Roll, listen for Player Elims to trigger the objective. If none happen, stop listening until we get another dodge.
      (Variables|Default|Setb_validElim true)
      (Utilities|IsValid _returnvalue
        (:Is%20Valid
          (bind _returnvalue_2 (EventProcessorSubsystem|StartListeningToFilteredEvent 0 _returnvalue (Variables|Default|Getverb_elim) true (delegate Custom|OnPawnEliminated)))
          (Variables|Default|Setevent_elim _returnvalue_2)
          (Utilities|FlowControl|Delay (Variables|Default|GetTime%20Limit))
          (EventProcessorSubsystem|StopListeningToFilteredEvent (Variables|Default|Getevent_elim))
          (Variables|Default|Setb_validElim false))))))

(event Custom|OnPawnEliminated
  ; Send verb message
  (CustomVerb|BroadcastCustomVerbMessage (CustomVerb|MakeCustomVerbMessage (Ability|GetOwningActorfromActorInfo) "(TagName=\"CustomVerb.CherrySmokeDualRevolver.DodgeElim\")"))
  (Ability|EndAbility))

(event Ability|EventOnAbilityRemoved
  ; Cleanup wait tasks
  (AsyncAction|Cancel (Variables|Default|Getasync_dodge))
  (EventProcessorSubsystem|StopListeningToFilteredEvent (Variables|Default|Getevent_elim)))
