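; Weekly quest ability that detects when the player exits Wraith Form near an enemy player.
; On activation it waits for the MistRaven GrimStrut end-ability gameplay event on the owner, then sphere-overlaps pawns within a configurable distance in meters.
; The first tag-matched player pawn off the local player's team triggers a custom verb message and ends the ability.
; Removing the ability cancels the async wait.
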(event Ability|EventActivateAbility
  (bind _asynctaskproxy (Ability|Async|WaitGameplayEventToActor (Ability|GetOwningActorfromActorInfo) "(TagName=\"MistRaven.GrimStrut.Ascended.EndAbility\")")
    (:then
      (Variables|Default|Setasync_mistend _asynctaskproxy))
    (:EventReceived
      ; Check if player exited Wraith Form near an enemy player
      (bind _fortpawn (Class|FortQuestBlueprintLibrary|GetFortPawnFromObject (Ability|GetOwningActorfromActorInfo))
        (:Valid
          (bind _outactors (Collision|SphereOverlapActors (Transformation|GetActorLocation _fortpawn) (* 1e+02 (Variables|Default|GetDistance%20Threshold%20in%20Meters)) (Utilities|Array|MakeArray "ObjectTypeQuery3") "/Script/FortniteGame.FortPawn" (Utilities|Array|MakeArray _fortpawn)))
          (for _array_element _outactors
            (bind _fortpawn_2 (Class|FortQuestBlueprintLibrary|GetFortPawnFromObject _array_element)
              (:Valid
                (if (GameplayTags|HasAnyTags (GameplayTags|GetOwnedGameplayTags _fortpawn_2 :!class /Script/GameplayTags.BlueprintGameplayTagLibrary) (Variables|Default|Gettc_players))
                  (if (not (Team|OnLocalPlayersTeam _fortpawn_2))
                    ; Send verb message
                    (CustomVerb|BroadcastCustomVerbMessage (CustomVerb|MakeCustomVerbMessage (Ability|GetOwningActorfromActorInfo) "(TagName=\"CustomVerb.WraithFormExit.NearEnemyPlayer\")"))
                    (Ability|EndAbility)))))))))))

(event Ability|EventOnAbilityRemoved
  ; Cleanup wait action
  (AsyncAction|Cancel (Variables|Default|Getasync_mistend)))
