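; Gameplay ability for Season 36 Battle Pass weekly quest W05 Q02 that reports hits from behind.
; On a gameplay event, it stores the instigator's player pawn and the event target, then compares their forward vectors by dot product.
; If both face roughly the same way (no position check), it broadcasts a CustomVerb.HitFromBehind message carrying the target's tags.
; The ability then ends.
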(event Ability|EventActivateAbilityFromEvent
  (params (EventData :type Gameplay%20Event%20Data%20Structure%20%28by%20ref%29))
  (bind (_eventtag _instigator _target) (Utilities|Struct|BreakGameplayEventData EventData))
  (Variables|Default|SetInstigator (Game|Player|GetPlayerPawn (Fortnite|GetFortPlayerControllerfromActor _instigator)))
  (Variables|Default|SetTarget _target)
  (if (>= (Math|Vector|DotProduct (Transformation|GetActorForwardVector (Variables|Default|GetTarget)) (Transformation|GetActorForwardVector (Variables|Default|GetInstigator))) 0.425)
    (bind _result (CustomVerb|AddTagsfromSubject (CustomVerb|MakeCustomVerbMessage (Ability|GetOwningActorfromActorInfo) "(TagName=\"CustomVerb.HitFromBehind\")") "Target" (Variables|Default|GetTarget)))
    (CustomVerb|BroadcastCustomVerbMessage _result)
    (else
      _))
  (Ability|EndAbility))
