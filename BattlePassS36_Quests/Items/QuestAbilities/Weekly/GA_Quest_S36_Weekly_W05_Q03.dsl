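; Quest ability for Season 36 Weekly W05 Q03 that detects when a player drops a minimum height without taking falling damage.
; On event activation it records the pawn's starting Z, then waits on both a falling-damage event, which ends the ability, and removal of the falling tag.
; When falling stops, a living pawn at least Height Check below the start broadcasts the HighFall custom verb.
; Ending the ability cleans up both wait tasks.
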(event Ability|EventActivateAbilityFromEvent
  (params (EventData :type Gameplay%20Event%20Data%20Structure%20%28by%20ref%29))
  ; Player started falling.
  (bind _fortpawn (Class|FortQuestBlueprintLibrary|GetFortPawnFromObject (Ability|GetOwningActorfromActorInfo))
    (:Valid
      (bind _output_get (Variables|Default|Setfp_player _fortpawn))
      (Variables|Default|SetActor%20Initial%20Height (.z (Transformation|GetActorLocation _output_get)))
      ; If falling damage received, bail.
      (bind _asynctaskproxy (Ability|Tasks|WaitGameplayEvent "(TagName=\"Event.Damage.Received.Falling\")")
        (:then
          (Variables|Default|Setasync_fallingHurt _asynctaskproxy)
          ; Wait until player is neither falling nor dead, then do a height check.
          (bind _asynctaskproxy_2 (Ability|Tasks|WaitGameplayTagRemove "(TagName=\"MovementMode.Falling\")")
            (:then
              (Variables|Default|Setasync_fallingOver _asynctaskproxy_2))
            (:Removed
              (if (not (FortPawn|IsDead (Variables|Default|Getfp_player)))
                (if (>= (- (Variables|Default|GetActor%20Initial%20Height) (Variables|Default|GetHeight%20Check)) (.z (Transformation|GetActorLocation (Variables|Default|Getfp_player))))
                  ; Send verb message
                  (CustomVerb|BroadcastCustomVerbMessage (CustomVerb|MakeCustomVerbMessage (Ability|GetOwningActorfromActorInfo) "(TagName=\"CustomVerb.HighFall\")"))
                  (Ability|EndAbility)
                  (else
                    (Ability|EndAbility)))
                (else
                  (Ability|EndAbility))))))
        (:EventReceived
          (Ability|EndAbility))))
    (:NotValid
      (Ability|EndAbility))))

(event Ability|EventOnEndAbility
  (params (bWasCancelled :type Boolean))
  ; Cleanup wait tasks
  (GameplayTasks|EndTask (Variables|Default|Getasync_fallingHurt))
  (GameplayTasks|EndTask (Variables|Default|Getasync_fallingOver)))
