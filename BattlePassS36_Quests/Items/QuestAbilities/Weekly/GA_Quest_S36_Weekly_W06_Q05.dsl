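; Season 36 weekly quest ability (W06 Q05) that reports hero rank changes for quest progress.
; On activation it listens on the global event router for OnHeroRankChanged and stores the async listener.
; When the owner resolves to a valid pawn and the payload rank matches the valid-rank tag container, it broadcasts the CherrySmokeHeroRank.RankS custom verb.
; Removing the ability stops the listener.
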(event Ability|EventActivateAbility
  ;[0.5 0.15 0.5 0.5] Listen for weird global Hero Rank Up event.
  (bind (_asynctaskproxy _payload) (GameplayEventRouter|StartListeningToEvent (GameplayEventRouter|GetGlobalEventRouter) "/Script/CherrySmokeHeroRankingSystemRuntime.OnHeroRankChanged")
    (:then
      (Variables|Default|Setasync_heroRank _asynctaskproxy))
    (:OnEventReceived
      ; Make sure it was us that ranked up, and to a valid rank for quest progress.
      (Class|FortQuestBlueprintLibrary|GetFortPawnFromObject (Ability|GetOwningActorfromActorInfo)
        (:Valid
          (if (GameplayTags|HasTag (Variables|Default|Gettc_validHeroRanks) (Utilities|Struct|BreakOnHeroRankChanged _payload) true)
            ; Send verb message
            (CustomVerb|BroadcastCustomVerbMessage (CustomVerb|MakeCustomVerbMessage (Ability|GetOwningActorfromActorInfo) "(TagName=\"CustomVerb.CherrySmokeHeroRank.RankS\")"))
            (Ability|EndAbility)))))))

(event Ability|EventOnAbilityRemoved
  ; Cleanup wait action
  (GameplayEventRouter|StopListeningtoEvent (Variables|Default|Getasync_heroRank)))
